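; Creative Channel Switchboard device that checks an instigating player against team, class, kills, score and team score requirements, then fires On Pass or On Fail.
; EvaluateMatch returns a per-check failure code or success, cached before the activation attempt and replicated when message transmission succeeds.
; OnRep shows channel or failure numbers and timed line feedback, while a non-cosmetic timer clears the result.
; Visuals, including enabled-state light colors, use custom primitive data.
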(event Custom|EvaluateChecksFromEvent
  (params (Instigator :type Controller%20Object%20Reference))
  (bind _asfort_player_controller (Utilities|Casting|CastToFortPlayerController Instigator))
  (bind _output_get (Variables|Default|SetCachedMatchResult (CallFunction|EvaluateMatch _asfort_player_controller)))
  (CallFunction|StartActivateAttempt Instigator (select (== _output_get (Variables|Default|GetSuccessMatchResultCode)) (delegate Custom|BroadcastChecksWereValid) (delegate Custom|BroadcastChecksWereInvalid))))

(event Custom|BroadcastChecksWereValid
  (params (Instigator :type Controller%20Object%20Reference))
  (GameplayEvent|BroadcastEvent (Variables|UserOptions|Events|GetOn%20Pass) 0 (Utilities|Struct|MakeGameplayEventTriggeronMatch Instigator)))

(event Custom|BroadcastChecksWereInvalid
  (params (Instigator :type Controller%20Object%20Reference))
  (GameplayEvent|BroadcastEvent (Variables|UserOptions|Events|GetOn%20Fail) 0 (Utilities|Struct|MakeGameplayEventTriggeronNoMatch Instigator)))

(event EventBeginPlay
  (bind (_disablecolorvfx_r _disablecolorvfx_g _disablecolorvfx_b _disablecolorvfx_a) (Math|Color|BreakColor (Variables|GetDisableColorVFX_R)))
  (bind (_enablecolorvfx_r _enablecolorvfx_g _enablecolorvfx_b _enablecolorvfx_a) (Math|Color|BreakColor (Variables|GetEnableColorVFX_R)))
  (bind (_delaycolorvfx_r _delaycolorvfx_g _delaycolorvfx_b _delaycolorvfx_a) (Math|Color|BreakColor (Variables|GetDelayColorVFX_R)))
  (bind _mesh (Variables|Default|GetMesh))
  (|CallParentFunction)
  (Variables|Default|SetCachedMinigameStatsComponent (Minigame|Stats|GetMinigameStatsComponent self))
  (Networking|CanExecuteCosmeticEvents
    (:True
      (switch Utilities|FlowControl|Switch|SwitchonEEnabledState (Variables|Default|GetCurrentState)
        (:NewEnumerator0
          (Rendering|Material|SetCustomPrimitiveDataFloat _mesh (Variables|VFX|GetLightFXBlinkDataIndex) 0.0)
          (Rendering|Material|SetCustomPrimitiveDataVector4 (Variables|Default|GetMesh) (Variables|VFX|GetLightFXColorDataIndex) (Math|Vector4|MakeVector4 _enablecolorvfx_r _enablecolorvfx_g _enablecolorvfx_b _enablecolorvfx_a)))
        (:NewEnumerator1
          (Rendering|Material|SetCustomPrimitiveDataFloat _mesh (Variables|VFX|GetLightFXBlinkDataIndex) 1.0)
          (Rendering|Material|SetCustomPrimitiveDataVector4 (Variables|Default|GetMesh) (Variables|VFX|GetLightFXColorDataIndex) (Math|Vector4|MakeVector4 _delaycolorvfx_r _delaycolorvfx_g _delaycolorvfx_b _delaycolorvfx_a)))
        (:NewEnumerator2
          (Rendering|Material|SetCustomPrimitiveDataVector4 (Variables|Default|GetMesh) (Variables|VFX|GetLightFXColorDataIndex) (Math|Vector4|MakeVector4 _disablecolorvfx_r _disablecolorvfx_g _disablecolorvfx_b _disablecolorvfx_a)))))))

(event Default|EventSetMaterialBasedOnCurrentState
  (bind (_disablecolorvfx_r _disablecolorvfx_g _disablecolorvfx_b _disablecolorvfx_a) (Math|Color|BreakColor (Variables|GetDisableColorVFX_R)))
  (bind (_enablecolorvfx_r _enablecolorvfx_g _enablecolorvfx_b _enablecolorvfx_a) (Math|Color|BreakColor (Variables|GetEnableColorVFX_R)))
  (bind (_delaycolorvfx_r _delaycolorvfx_g _delaycolorvfx_b _delaycolorvfx_a) (Math|Color|BreakColor (Variables|GetDelayColorVFX_R)))
  (bind _mesh (Variables|Default|GetMesh))
  (Networking|CanExecuteCosmeticEvents
    (:True
      (switch Utilities|FlowControl|Switch|SwitchonEEnabledState (Variables|Default|GetCurrentState)
        (:NewEnumerator0
          (Rendering|Material|SetCustomPrimitiveDataFloat _mesh (Variables|VFX|GetLightFXBlinkDataIndex) 0.0)
          (Rendering|Material|SetCustomPrimitiveDataVector4 (Variables|Default|GetMesh) (Variables|VFX|GetLightFXColorDataIndex) (Math|Vector4|MakeVector4 _enablecolorvfx_r _enablecolorvfx_g _enablecolorvfx_b _enablecolorvfx_a)))
        (:NewEnumerator1
          (Rendering|Material|SetCustomPrimitiveDataFloat _mesh (Variables|VFX|GetLightFXBlinkDataIndex) 1.0)
          (Rendering|Material|SetCustomPrimitiveDataVector4 (Variables|Default|GetMesh) (Variables|VFX|GetLightFXColorDataIndex) (Math|Vector4|MakeVector4 _delaycolorvfx_r _delaycolorvfx_g _delaycolorvfx_b _delaycolorvfx_a)))
        (:NewEnumerator2
          (Rendering|Material|SetCustomPrimitiveDataVector4 (Variables|Default|GetMesh) (Variables|VFX|GetLightFXColorDataIndex) (Math|Vector4|MakeVector4 _disablecolorvfx_r _disablecolorvfx_g _disablecolorvfx_b _disablecolorvfx_a)))))))

(event Custom|MatchResultFeedbackStart
  (CallFunction|UpdateChannelNumberwithMatchResult (Variables|Replicated|GetLastMatchResult)))

(event Custom|CleanupLastResult
  (Variables|Replicated|SetLastMatchResult (Variables|Default|GetInvalidMatchResultCode))
  (Utilities|Time|InvalidateTimerHandle (Variables|Default|GetCleanupTimerHandle)))

(event EventDestroyed
  (Networking|CanExecuteCosmeticEvents
    (:False
      (Utilities|Time|ClearandInvalidateTimerbyHandle (Variables|Default|GetCleanupTimerHandle)))))

(event ChannelNumberUpdate|EventUpdateChannelNumber
  (params (Num :type Integer)))

(event EventSwitchtoDyamicMaterial
  (Networking|CanExecuteCosmeticEvents
    (:True
      (if (== (Variables|Replicated|GetLastMatchResult) (Variables|Default|GetInvalidMatchResultCode))
        (CallFunction|ShowListenChannel)
        (else
          (CallFunction|MatchResultFeedbackStart))))))

(event EventSetCurrentConstantMaterial
  (params (NewCurrentMaterial :type Material%20Interface%20Object%20Reference)))

(event OnChannelIdsReplicated(ListenToChannel)
  (CallFunction|ShowListenChannel))

(event OnAnyPropertyChanged(ToyOptionsComponent)
  (CallFunction|ShowListenChannel))

(event OnPropertiesLoaded(ToyOptionsComponent)
  (CallFunction|ShowListenChannel))

(fn EvaluateMatch
  (params (FortPlayerController :type Fort%20Player%20Controller%20Object%20Reference))
  (bind _localcurrentminigame (LocalVariables|GetLocalCurrentMinigame))
  (bind _returnvalue (Minigame|GetMinigameTeamComponent _localcurrentminigame))
  (Utilities|FlowControl|Sequence
    (:then_0
      (if (Knobs|Team|IsTeamAny (Variables|UserKnobs|GetCheckTeam))
        _
        (elif (== (Knobs|Team|GetTeamValueasIndex (Variables|UserKnobs|GetCheckTeam)) (Math|Conversions|ToInteger(Byte) (Team|GetActorPvPTeamIndex FortPlayerController)))
          _
          (else
            (return 1)))))
    (:then_1
      (bind _fortminigamelogic (Variables|Default|GetFortMinigameLogic)
        (:then
          (bind _currentminigame (Variables|/Script/FortniteGame.FortMinigameLogicComponent|GetCurrentMinigame _fortminigamelogic)
            (:then
              (bind _output_get (LocalVariables|SetLocalCurrentMinigame _currentminigame))
              (Utilities|IsValid (Minigame|GetFNEClassComponent _output_get)
                (:Is%20Valid
                  (LocalVariables|SetLocalMinigameFNEClassComponent (Minigame|GetFNEClassComponent _output_get))
                  (Utilities|IsValid _returnvalue
                    (:Is%20Valid
                      (LocalVariables|SetLocalMinigameTeamComponent _returnvalue)
                      (bind _asfort_player_state (Utilities|Casting|CastToFortPlayerState (Variables|/Script/Engine.Controller|GetPlayerState (Variables|GetFortPlayerController)))
                        (:then
                          (LocalVariables|SetLocalFortPlayerState _asfort_player_state)
                          (Variables|Default|GetCachedMinigameStatsComponent
                            (:then)
                            (:else
                              (return (Variables|Default|GetInvalidMatchResultCode)))))
                        (:CastFailed
                          _)))))
                (:Is%20Not%20Valid
                  _))
              (return (Variables|Default|GetInvalidMatchResultCode)))
            (:else
              (return (Variables|Default|GetInvalidMatchResultCode)))))
        (:else
          (return (Variables|Default|GetInvalidMatchResultCode)))))
    (:then_2
      (if (== (Variables|UserKnobs|GetCheckClass) -1)
        _
        (elif (Math|Byte|Equal(Byte) (Minigame|GetClassSlotforPlayer (LocalVariables|GetLocalMinigameFNEClassComponent) (LocalVariables|GetLocalFortPlayerState) true) (Math|Conversions|ToByte(Integer) (Variables|UserKnobs|GetCheckClass)))
          _
          (else
            (return 2)))))
    (:then_3
      (if (!= (Variables|UserKnobs|GetMinKills) 0)
        (if (< (Minigame|Stats|GetStatforPlayer (Variables|Default|GetCachedMinigameStatsComponent) (LocalVariables|GetLocalFortPlayerState) (Class|FortMinigameStatsComponent|GetLoadedStatfromClass (Variables|Default|GetCachedMinigameStatsComponent) "/Game/Athena/Playset/Minigames/StatFilters/MinigameStat_PlayerKills.MinigameStat_PlayerKills_C") false) (Variables|UserKnobs|GetMinKills))
          (return 3))))
    (:then_4
      (if (< (Minigame|Stats|GetStatforPlayer (Variables|Default|GetCachedMinigameStatsComponent) (LocalVariables|GetLocalFortPlayerState) (Class|FortMinigameStatsComponent|GetLoadedStatfromClass (Variables|Default|GetCachedMinigameStatsComponent) "/Game/Athena/Playset/Minigames/StatFilters/MinigameStat_Score.MinigameStat_Score_C") false) (Variables|UserKnobs|GetMinScore))
        (return 4)))
    (:then_5
      (if (== (Minigame|Teams|GetMinigameNumTeams (LocalVariables|GetLocalMinigameTeamComponent)) (LocalVariables|GetCooperative))
        (for _array_element (Minigame|GetParticipatingPlayers (LocalVariables|GetLocalCurrentMinigame))
          (Utilities|IsValid _array_element
            (:Is%20Valid
              (LocalVariables|SetCooperativeScore (+ (LocalVariables|GetCooperativeScore) (Minigame|Stats|GetStatforPlayerBucket (Variables|Default|GetCachedMinigameStatsComponent) (Minigame|GetBucketforPlayerState (LocalVariables|GetLocalMinigameTeamComponent) _array_element false) (Class|FortMinigameStatsComponent|GetLoadedStatfromClass (Variables|Default|GetCachedMinigameStatsComponent) "/Game/Athena/Playset/Minigames/StatFilters/MinigameStat_Score.MinigameStat_Score_C") false false))))))
        (if (< (LocalVariables|GetCooperativeScore) (Variables|UserKnobs|GetMinTeamScore))
          (return 5))
        (else
          (if (< (Minigame|Stats|GetStatforPlayerBucket (Variables|Default|GetCachedMinigameStatsComponent) (Minigame|GetBucketforPlayerState (LocalVariables|GetLocalMinigameTeamComponent) (LocalVariables|GetLocalFortPlayerState) true) (Class|FortMinigameStatsComponent|GetLoadedStatfromClass (Variables|Default|GetCachedMinigameStatsComponent) "/Game/Athena/Playset/Minigames/StatFilters/MinigameStat_Score.MinigameStat_Score_C") false false) (Variables|UserKnobs|GetMinTeamScore))
            (return 5)))))
    (:then_6
      (return (Variables|Default|GetSuccessMatchResultCode)))))

(fn GetFeedbackNumberByMatchResultCode
  (params (MatchResultCode :type Integer))
  (if (Class|CreativeBlueprintLibrary|IsDirectBindingEnabled (Minigame|GetMinigameSettingsMachine self))
    (return (Variables|Default|GetUseDirectBindingResultCode))
    (elif (== (Variables|GetMatchResultCode) (Variables|Default|GetSuccessMatchResultCode))
      (return (Math|Integer|SelectInt (Variables|Default|GetNo%20ChannelMatchResultCode) (Class|FortGameplayMessageComponentBase|GetMessageChannelIndex (Variables|Default|GetMatchTransmitOn)) (== (Class|FortGameplayMessageComponentBase|GetMessageChannelIndex (Variables|Default|GetMatchTransmitOn)) 0)))
      (elif (== (Variables|GetMatchResultCode) (Variables|Default|GetInvalidMatchResultCode))
        (return (Math|Integer|SelectInt (Variables|Default|GetNo%20ChannelMatchResultCode) (Class|FortGameplayMessageComponentBase|GetMessageChannelIndex (Variables|Default|GetListenToChannel)) (== (Class|FortGameplayMessageComponentBase|GetMessageChannelIndex (Variables|Default|GetListenToChannel)) 0)))
        (else
          (return (Variables|GetMatchResultCode)))))))

(fn GetTeamSettingValue
  (return (Knobs|Team|GetShiftedAnyTeamValue)))

(fn OnRep_LastMatchResult
  (Networking|CanExecuteCosmeticEvents
    (:True
      (if (== (Variables|Replicated|GetLastMatchResult) (Variables|Default|GetInvalidMatchResultCode))
        (CallFunction|ShowListenChannel)
        (else
          (CallFunction|MatchResultFeedbackStart))))
    (:False
      (Utilities|FlowControl|Sequence
        (:then_0
          (if (Utilities|Time|IsValidTimerHandle (Variables|Default|GetCleanupTimerHandle))
            (Utilities|Time|ClearandInvalidateTimerbyHandle (Variables|Default|GetCleanupTimerHandle))))
        (:then_1
          (bind _returnvalue (Utilities|Time|SetTimerbyEvent (EventDispatchers|CreateEvent :!event CleanupLastResult) (Math|Float|Max(Float) (* (Variables|Replication|GetNetUpdateFrequency) 1.5) (Variables|VFX|GetMatchResultFeedbackTime))))
          (Variables|Default|SetCleanupTimerHandle _returnvalue))))))

(fn ShowListenChannel
  (CallFunction|UpdateChannelNumberwithMatchResult (Variables|Default|GetInvalidMatchResultCode)))

(fn TransmitMessage
  (bind _success (|CallParentFunction))
  (if _success
    (Variables|Replicated|SetLastMatchResult (Variables|Default|GetCachedMatchResult)))
  (return _success))

(fn UpdateChannelNumberWithMatchResult
  (params (MatchResultCode :type Integer))
  (Rendering|Material|SetCustomPrimitiveDataFloat (Variables|Default|GetMesh) (Variables|VFX|GetChannelFXDataIndex) (Math|Conversions|ToFloat(Integer) (CallFunction|GetFeedbackNumberbyMatchResultCode MatchResultCode)))
  (if (== (Variables|GetMatchResultCode) (Variables|Default|GetInvalidMatchResultCode))
    (Rendering|Material|SetCustomPrimitiveDataVector4 (Variables|Default|GetMesh) (Variables|VFX|GetLineFXDataIndex) (0.0 0.0 0.0 0.0))
    (else
      (Rendering|Material|SetCustomPrimitiveDataVector4 (Variables|Default|GetMesh) (Variables|VFX|GetLineFXDataIndex) (Math|Vector4|MakeVector4 (Utilities|Time|GetTimeSeconds) (+ (Utilities|Time|GetTimeSeconds) (Variables|VFX|GetMatchResultFeedbackTime)) (Variables|VFX|GetLineFadeTime) (select (== (Variables|GetMatchResultCode) (Variables|Default|GetSuccessMatchResultCode)) 1.0 -1.0))))))

(fn ConstructionScript)
