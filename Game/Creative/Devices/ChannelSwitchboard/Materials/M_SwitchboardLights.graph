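# Lit material for the Channel Switchboard device that drives its glowing True, False and receptor indicator lights alongside diffuse, normal and mask textures.
# Vertex color G, B and R mask the three lights, which are summed.
# Custom primitive data supplies a timestamp for the MF_Timestamp_InOut_01 flash, with alpha's sign choosing True or False.
# The receptor takes its color from primitive data and pulses on a sine while waiting to transmit.
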Material /Game/Creative/Devices/ChannelSwitchboard/Materials/M_SwitchboardLights
graph {
 properties {
  pixelDepthOffsetMode=PDOM_Legacy
  previewMesh=/Game/Creative/Devices/ChannelSwitchboard/Switchboard/Meshes/SM_Switchboard
  usedShadingModels=Default Lit
 }
 nodes [
  n1 = VectorParameter() {
   bUseCustomPrimitiveData=true
   desc="R = on transition start time\r\nG = off transition start time\r\nB = transition length"
   parameterName=Timestamp
   primitiveDataIndex=28
  }
  n2 = Saturate(None=@n1.A)
  n3 = ScalarParameter() {defaultValue=15, parameterName=Emissive Strength}
  n4 = Multiply(
   A=MaterialFunctionCall(TimestampA=@n1.R, TransitionLength_In=@n1.R, TimestampB=@n1.R, TransitionLength_Out=Constant(), Transition In Only?=StaticBool()) {materialFunction=&MF_Timestamp_InOut_01}.Linear
   B=@n3
  )
  n5 = ScalarParameter() {defaultValue=1, parameterName=Light Off Intensity}
  n6 = VertexColor()
  n7 = Multiply(A=Add(A=Multiply(A=@n2, B=@n4), B=@n5), B=@n6.G)
  n8 = VectorParameter() {parameterName=True Light, defaultValue=<rgba 0.017 1 0.1853 0>}
  n9 = Saturate(None=Multiply(A=@n1.A) {constB=-1})
  n10 = Multiply(A=Add(A=Multiply(A=@n9, B=@n4), B=@n5), B=@n6.B)
  n11 = VectorParameter() {parameterName=False Light, defaultValue=<rgba 2 0.1 0.1 0>}
  n12 = Add(A=@n4, B=@n5)
  n13 = VectorParameter() {bUseCustomPrimitiveData=true, parameterName=Receptor Light, primitiveDataIndex=24, defaultValue=<rgba 1 0.4512 0 0>}
  n14 = Time()
  n15 = Multiply(A=Add(A=Sine(None=@n14)) {constB=1.5}) {constB=0.6667}
  n16 = ScalarParameter() {
   bUseCustomPrimitiveData=true
   defaultValue=1
   desc="0 = FALSE\r\n1 = TRUE"
   parameterName=Waiting to transmit
   primitiveDataIndex=14
   sliderMax=1
   sliderMin=-1
  }
  n17 = Multiply(A=@n13.RGB, B=LinearInterpolate(B=Multiply(A=@n3, B=@n15), Alpha=@n16) {constA=1})
  Add(A=Add(A=Multiply(A=@n7, B=@n8.RGB), B=Multiply(A=@n10, B=@n11.RGB)), B=Multiply(A=@n6.R, B=Multiply(A=@n12, B=@n17)))
  ScalarParameter() {defaultValue=1, desc="1 = True   /   0 = None   /   -1 = False", parameterName=Output, primitiveDataIndex=11, sliderMax=1, sliderMin=-1}
  ScalarParameter() {defaultValue=1, parameterName=DEBUG}
  TextureSample() {texture=/Game/Creative/Devices/ChannelSwitchboard/Textures/T_Switchboard_D}
  TextureSample() {samplerType=SAMPLERTYPE_Normal, texture=/Game/Creative/Devices/ChannelSwitchboard/Textures/T_Switchboard_N}
  TextureSample() {samplerType=SAMPLERTYPE_Masks, texture=/Game/Creative/Devices/ChannelSwitchboard/Textures/T_Switchboard_S}
 ]
}
