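# Screen material for the Creative Channel Switchboard device, showing its channel, output state and enabled state from custom primitive data.
# A timestamp function times a horizontal wipe on screen masks and, with output type, selects true/false labels and colors.
# An atlas supplies labels and three channel digits, replaced by a no-channel icon at channel zero.
# Disabled screens use error colors, and animated scanlines with a scrolling band fill the background.
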Material /Game/Creative/Devices/ChannelSwitchboard/Materials/M_SwitchboardScreen
graph {
 properties {pixelDepthOffsetMode=PDOM_Legacy, usedShadingModels=Default Lit}
 nodes [
  n1 = VectorParameter() {
   bUseCustomPrimitiveData=true
   desc="R = on transition start time\r\nG = off transition start time\r\nB = transition length\r\nA = output type \r\n           -1 = false\r\n           0   = none\r\n           1   = true"
   parameterName=Timestamp
   primitiveDataIndex=28
  }
  n2 = MaterialFunctionCall(TimestampA=@n1.R, TransitionLength_In=@n1.R, TimestampB=@n1.R, TransitionLength_Out=Constant(), Transition In Only?=StaticBool()) {materialFunction=&MF_Timestamp_InOut_01}
  n3 = TextureCoordinate()
  n4 = Sign(None=Subtract(A=@n2.Linear, B=ComponentMask(None=@n3) {r=true}))
  n5 = Multiply(A=@n2.Linear, B=@n1.A)
  n6 = TextureSample() {samplerType=SAMPLERTYPE_LinearColor, texture=/Game/Creative/Devices/ChannelSwitchboard/Textures/T_Switchboard_Screen_M}
  n7 = Multiply(A=Saturate(None=@n5), B=@n6.G)
  n8 = Multiply(A=Saturate(None=Multiply(A=@n5) {constB=-1}), B=@n6.R)
  n9 = Add(A=Multiply(A=Saturate(None=@n4), B=Add(A=@n7, B=@n8)), B=Multiply(A=@n6.B) {constB=0.05})
  n10 = VectorParameter() {parameterName=Text Error Color, defaultValue=<rgba 3 0 0 1>}
  n11 = Constant() {r=1}
  n12 = Saturate(None=Multiply(A=@n5) {constB=-1, desc=when is -1})
  n13 = LinearInterpolate(
   A=LinearInterpolate(
    A=VectorParameter() {parameterName=Text Color, defaultValue=<rgba 0 1.8 4 1>}.RGB
    B=VectorParameter() {parameterName=Text True Color, defaultValue=<rgba 0 7 3.8 1>}.RGB
    Alpha=Saturate(None=@n5)
   )
   B=@n10.RGB
   Alpha=If(A=ScalarParameter() {bUseCustomPrimitiveData=true, parameterName=Channel, primitiveDataIndex=10}, "A > B"=@n12, "A == B"=@n11, "A < B"=@n11) {constB=-1}
  ) {constAlpha=0}
  n14 = ScalarParameter() {bUseCustomPrimitiveData=true, parameterName=IsEnabled, primitiveDataIndex=20}
  n15 = LinearInterpolate(A=@n10.RGB, B=@n13, Alpha=@n14)
  n16 = VectorParameter() {parameterName=UVs Scale / Offset, defaultValue=<rgba 0.35 0.35 0.25 -0.07>}
  n17 = MaterialFunctionCall(UVs=Add(A=TextureCoordinate(), B=AppendVector(A=@n16.B, B=@n16.B)), Texture Scale=AppendVector(A=@n16.R, B=@n16.R)) {materialFunction=&ScaleUVsByCenter}
  n18 = Multiply(A=@n17.UVs, B=Constant2Vector() {g=2, r=1})
  n19 = TextureSample(UVs=Multiply(A=@n18, B=Constant2Vector() {g=1, r=0.3})) {samplerType=SAMPLERTYPE_LinearColor, texture=/Game/Creative/Devices/ChannelSwitchboard/Textures/T_CreativeTrigger_NumberSet_L_ERROR}
  n20 = Sign(None=Add(A=@n5) {constB=0.5, desc=when is 0 or above})
  n21 = Add(A=Multiply(A=@n19.G, B=Saturate(None=@n20)), B=Multiply(A=@n19.B, B=@n12))
  n22 = Ceil(None=ComponentMask(None=Subtract(A=@n17.UVs) {constB=0.5}) {g=true})
  n23 = ScalarParameter() {bUseCustomPrimitiveData=true, parameterName=Channel, primitiveDataIndex=10}
  n24 = Floor(None=Multiply(A=@n23) {constB=0.01})
  n25 = Multiply(A=@n18, B=Constant2Vector() {g=1, r=0.3})
  n26 = ComponentMask(None=@n25) {r=true}
  n27 = ComponentMask(None=@n25) {g=true}
  n28 = AppendVector(A=Add(A=Multiply(A=@n24) {constB=0.1}, B=@n26), B=@n27)
  n29 = Multiply(A=Floor(None=Multiply(A=@n23) {constB=0.1})) {constB=0.1}
  n30 = AppendVector(A=Add(A=Add(A=@n29, B=@n26)) {constB=-0.1}, B=@n27)
  n31 = ComponentMask(None=Subtract(A=@n18) {constB=0.3333}) {r=true}
  n32 = Floor(None=@n23)
  n33 = Add(A=Add(A=Multiply(A=@n32) {constB=0.1}, B=@n26)) {constB=-0.2}
  n34 = Subtract(A=@n18) {constB=0.6667}
  n35 = LinearInterpolate(
   A=LinearInterpolate(
    A=TextureSample(UVs=@n28) {samplerType=SAMPLERTYPE_LinearColor, texture=&T_CreativeTrigger_NumberSet_L_ERROR}.R
    B=TextureSample(UVs=@n30) {samplerType=SAMPLERTYPE_LinearColor, texture=&T_CreativeTrigger_NumberSet_L_ERROR}.R
    Alpha=Ceil(None=@n31)
   )
   B=TextureSample(UVs=AppendVector(A=@n33, B=@n27)) {samplerType=SAMPLERTYPE_LinearColor, texture=&T_CreativeTrigger_NumberSet_L_ERROR}.R
   Alpha=Ceil(None=ComponentMask(None=@n34) {r=true})
  )
  n36 = ScalarParameter() {bUseCustomPrimitiveData=true, parameterName=Channel, primitiveDataIndex=10}
  n37 = MaterialFunctionCall() {materialFunction=&LinearGradient}
  n38 = Constant() {r=1.6}
  n39 = LinearInterpolate(
   A=MaterialFunctionCall(In=Power(Base=@n37.VGradient, Exp=@n38), Contrast=Constant() {r=100}) {materialFunction=&CheapContrast}.Result
   Alpha=ScalarParameter() {bUseCustomPrimitiveData=true, parameterName=IsEnabled, primitiveDataIndex=20}
  ) {constA=1}
  n40 = TextureCoordinate() {uTiling=2, vTiling=2}
  n41 = Constant() {r=1.4}
  n42 = TextureSample(UVs=Add(A=Multiply(A=@n40, B=@n41), B=Constant2Vector() {g=-1.1, r=-0.23})) {
   samplerSource=SSM_Clamp_WorldGroupSettings
   samplerType=SAMPLERTYPE_Masks
   texture=/Game/Creative/Devices/Trigger/Materials/T_Device_Trigger_NoChannel_M
  }
  n43 = Multiply(A=LinearInterpolate(B=@n42.R, Alpha=@n42.R))
  n44 = Multiply(A=LinearInterpolate(A=Multiply(A=@n39, B=@n43), B=LinearInterpolate(A=@n21, B=@n35, Alpha=@n22), Alpha=Saturate(None=@n36)), B=@n17.0-1 mask)
  n45 = LinearInterpolate(A=VectorParameter() {parameterName=Disabled Icon Color, defaultValue=<rgba 1 0 0 1>}.RGB, B=@n13, Alpha=@n14)
  n46 = Time()
  n47 = Multiply(A=@n46) {constB=-0.05}
  n48 = Multiply(A=ComponentMask(None=TextureCoordinate()) {g=true}) {constB=1.2}
  n49 = ConstantBiasScale(None=Sine(None=Add(A=@n47, B=@n48)) {period=0.05})
  n50 = Multiply(A=@n46) {constB=-0.08}
  n51 = Power(Base=Frac(None=Add(A=@n50, B=@n48))) {constExponent=8}
  n52 = Multiply(A=@n15) {constB=0.03}
  n53 = Multiply(A=Add(A=Multiply(A=@n49) {constB=0.1}, B=Multiply(A=@n51) {constB=0.3}), B=@n52)
  Add(A=Add(A=Multiply(A=@n9, B=@n15), B=Multiply(A=@n44, B=@n45)), B=Multiply(A=Add(A=@n53, B=@n52), B=OneMinus(None=@n44)))
 ]
}
